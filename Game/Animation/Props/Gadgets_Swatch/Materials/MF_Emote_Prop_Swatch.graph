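# Material function for the Gadgets Swatch emote prop that overrides base color, emissive, metallic, specular and roughness on incoming material attributes.
# Pre-skinned position yields radial coordinates and a distance mask via named reroutes.
# Time-scrolled UVs, warped by gradient-path textures, sample two blended gradients, using explicit mesh-UV derivatives for mip selection.
# Emissive adds a tinted Fresnel edge to those gradients or an alternate contrast-mapped two-color look, selected by a disconnect-emote parameter.
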MaterialFunction /Game/Animation/Props/Gadgets_Swatch/Materials/MF_Emote_Prop_Swatch
graph {nodes [
  n1 = VectorParameter() {
   group=EmoteSpecific
   parameterName=Fresnel Edge Color and multiply
   channelNames {a=Multiply, b="", g="", r=""}
   defaultValue=<rgba 1 0 0.0787 10>
  }
  n2 = Multiply(
   A=Fresnel(
    ExponentIn=ScalarParameter() {defaultValue=4, group=EmoteSpecific, parameterName=Fresnel EXP}
    BaseReflectFractionIn=ScalarParameter() {defaultValue=0.01, group=EmoteSpecific, parameterName=BaseReflectFractionIn}
   ) {baseReflectFraction=0.01, exponent=4}
   B=@n1.Multiply
  ) {constB=10}
  n3 = NamedRerouteUsage()
  n4 = Time() {period=1}
  n5 = VectorParameter() {group=EmoteSpecific, parameterName=UV Speeds, channelNames {a=UV2 Y, b=UV2 X, g=UV1 Y, r=UV1 X}, defaultValue=<rgba -0.2 1.2 -0.1 -0.8>}
  n6 = Frac(None=Multiply(A=@n4, B=@n5.UV1 X) {constB=-0.2})
  n7 = Frac(None=Multiply(A=@n4, B=@n5.UV1 Y) {constB=1.2})
  n8 = NamedRerouteDeclaration(None=Add(A=Multiply(A=@n3) {constB=0.1}, B=AppendVector(A=@n6, B=@n7))) {name=uv move, nodeColor=<rgba 0 0.625 1 1>}
  n9 = ComponentMask(None=@n8) {g=true}
  n10 = TextureSampleParameter2D(UVs=Add(A=TextureCoordinate(), B=Constant2Vector() {g=0.03})) {
   group=EmoteSpecific
   parameterName=Emote Gradiant Path
   samplerType=SAMPLERTYPE_LinearColor
   texture=/Game/Animation/Props/Gadgets_Swatch/Textures/T_Emote_Prop_Swatch_M2
  }
  n11 = Multiply(A=@n10.R) {constB=-0.8}
  n12 = TextureCoordinate()
  n13 = NamedRerouteUsage()
  n14 = Time() {period=1}
  n15 = AppendVector(A=Frac(None=Multiply(A=@n14, B=@n5.UV2 X) {constB=-0.1}), B=Frac(None=Multiply(A=@n14, B=@n5.UV2 Y) {constB=-0.8}))
  n16 = TextureSampleParameter2D!(
   UVs=NamedRerouteDeclaration(None=Add(A=@n13, B=@n15)) {name=uv move2, nodeColor=<rgba 0 0.625 1 1>}
   "DDX(UVs)"=DDX(Value=@n12)
   "DDY(UVs)"=DDY(Value=@n12)
   Apply View MipBias
  ) {
   mipValueMode=TMVM_Derivative
   parameterName=Gradient path modifier
   samplerType=SAMPLERTYPE_LinearGrayscale
   texture=/Game/Animation/Props/Gadgets_Swatch/Textures/T_Emote_Prop_Swatch_M
  }
  n17 = NamedRerouteDeclaration(None=@n16.R) {name=UVFx, nodeColor=<rgba 1 0.0703 0 1>}
  n18 = AppendVector(A=ComponentMask(None=@n8) {r=true}, B=Add(A=Add(A=@n9, B=@n11), B=Multiply(A=@n17) {constB=0.5}))
  n19 = TextureCoordinate()
  n20 = DDX(Value=@n19)
  n21 = DDY(Value=@n19)
  n22 = TextureSampleParameter2D(UVs=@n18, "DDX(UVs)"=@n20, "DDY(UVs)"=@n21) {
   group=EmoteSpecific
   mipValueMode=TMVM_Derivative
   parameterName=Gradiant A
   texture=/Game/Animation/Props/Gadgets_Swatch/Textures/T_Emote_Prop_Swatch_Tex4
  }
  n23 = TextureSampleParameter2D(UVs=@n18, "DDX(UVs)"=@n20, "DDY(UVs)"=@n21) {
   group=EmoteSpecific
   mipValueMode=TMVM_Derivative
   parameterName=Gradiant B
   texture=/Game/Animation/Props/Gadgets_Swatch/Textures/T_Emote_Prop_Swatch_Tex2
  }
  n24 = LinearInterpolate(A=@n22.RGB, B=@n23.RGB, Alpha=Power(Base=NamedRerouteUsage()) {constExponent=6}) {constB=2}
  n25 = VectorParameter() {
   group=EmoteSpecific
   parameterName=Alternate Mask Options
   channelNames {a=Contrast, b=Add, g=Multiply, r=Power}
   defaultValue=<rgba 1 1 0 0>
  }
  n26 = Multiply(A=Power(Base=@n10.R, Exp=@n25.Power), B=@n25.Multiply)
  n27 = LinearInterpolate(
   A=VectorParameter() {group=EmoteSpecific, parameterName=Alternate Color Low}.RGB
   B=VectorParameter() {group=EmoteSpecific, parameterName=Alternate Color High, defaultValue=<rgba 1 1 1 0>}.RGB
   Alpha=MaterialFunctionCall(In=Add(A=@n26, B=@n25.Add), Contrast=@n25.Contrast) {materialFunction=&CheapContrast}.Result
  )
  n28 = ScalarParameter() {group=EmoteSpecific, parameterName=Use Alternative For Disconnect Emote}
  n29 = SetMaterialAttributes!(
   MaterialAttributes=FunctionInput() {inputType=FunctionInput_MaterialAttributes}
   BaseColor=VectorParameter() {group=EmoteSpecific, parameterName=Base Color}.RGB
   EmissiveColor=Add(A=Multiply(A=@n2, B=@n1), B=LinearInterpolate(A=@n24, B=@n27, Alpha=@n28))
   Metallic=ScalarParameter() {group=EmoteSpecific, parameterName=Metallic}
   Specular=ScalarParameter() {group=EmoteSpecific, parameterName=Specular}
   Roughness=ScalarParameter() {defaultValue=1, group=EmoteSpecific, parameterName=Roughness}
  ) {attributeSetTypes [
    #1
    #2
    #3
    #4
    #5
   ]}
  FunctionOutput(None=@n29)
  n30 = PreSkinnedPosition()
  n31 = Constant3Vector() {constant=<rgba 0.1 0 -23 0>}
  n32 = VertexInterpolator(VS=Divide(A=Add(A=@n30, B=@n31), B=Constant3Vector() {constant=<rgba 19.5 1 23.05 0>}))
  NamedRerouteDeclaration(None=MaterialFunctionCall(Vector or UVs=ComponentMask(None=@n32.PS) {b=true, r=true}) {materialFunction=&VectorToRadialValue}.Radial Coordinates) {name=uv, nodeColor=<rgba 1 0 0.8438 1>}
  n33 = ComponentMask(None=@n32.PS) {b=true, r=true}
  NamedRerouteDeclaration(None=Distance(A=Multiply(A=@n33) {constB=1.5}, B=Constant())) {name=mask, nodeColor=<rgba 0 0.6484 1 1>}
  n34 = WorldPosition()
  TransformPosition(None=@n34.XYZ) {transformSourceType=TRANSFORMPOSSOURCE_Instance, transformType=TRANSFORMPOSSOURCE_World}
  TransformPosition(None=@n34.XYZ) {transformSourceType=TRANSFORMPOSSOURCE_World, transformType=TRANSFORMPOSSOURCE_Instance}
 ]}
